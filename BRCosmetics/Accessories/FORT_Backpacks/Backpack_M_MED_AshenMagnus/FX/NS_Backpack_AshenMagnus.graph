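# Niagara effect for the Ashen Magnus back bling, adding fans and glows that respond to a reactivity value.
# Fans spawns eight socket-attached mesh particles that spin by User.Reactivity, while System.ReactivityDebug drives glow colors and alpha.
# Main_Light pairs a light with a sprite squashed by camera angle, and Wing_Lights places partitioned glow sprites on sockets.
# User.Variant switches the palette, mainly green or red.
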NiagaraSystem /BRCosmetics/Accessories/FORT_Backpacks/Backpack_M_MED_AshenMagnus/FX/NS_Backpack_AshenMagnus
dependencies {
 usedDataInterfaces [&NiagaraDataInterfaceCamera, &NiagaraDataInterfaceCurve, &NiagaraDataInterfaceSocketReader]
 usedDynamicInputs []
 usedModules [
  &self:SystemSpawnScript.SystemScriptSource.SystemScriptGraph.NiagaraNodeAssignment_3.SetVariables_#1
  &SystemState
  &self:SystemSpawnScript.SystemScriptSource.SystemScriptGraph.NiagaraNodeAssignment_0.SetVariables_#2
  &self:SystemSpawnScript.SystemScriptSource.SystemScriptGraph.NiagaraNodeAssignment_4.SetVariables_#3
 ]
}
emitters {
 Fans {
  emitterUpdateScript [
   &EmitterState(Life Cycle Mode=::NewEnumerator1, Loop Behavior=::NewEnumerator1, Loop Duration=0.1, UseLoopDelay=false)
   &SpawnBurst_Instantaneous(Spawn Count=8, Use Spawn Probability=false, UseLoopCountLimit=false)
  ]
  particleSpawnScript [
   &FN_SetSpawnGroup()
   &InitializeParticle(
    Color=<rgba 0.0035 0.005 0.0029 0.8>
    Lifetime=0.5
    Mesh Scale Mode=::NewEnumerator3
    Sprite Rotation Mode=::NewEnumerator2
    Sprite Size Mode=::NewEnumerator3
    Uniform Sprite Size=12
    UsePositionOffset=false
    Write Material Random=true
   )
   /SetVariables(Particles.GlowCameraOffset=5) {isSetParameters=true}
  ]
  particleUpdateScript [
   &ParticleState(Kill Particles When Lifetime Has Elapsed=false)
   &SocketLocation(
    Orientation Sampling=::NewEnumerator1
    SamplePosOffset=true
    Socket Index=$Particles.SpawnGroup
    Socket Reader=$System.Socket Reader
    Socket Sampling Mode=::NewEnumerator3
    VelocitySpeedLimit=false
   )
   &UpdateMeshOrientation(Rotation Rate=&Lerp_Float(A=-5, B=1000, Alpha=$User.Reactivity), Rotation Vector=<0 1 0>)
   &ScaleColor(
    Scale Alpha=&FloatFromCurve(
     FloatCurve=DataInterface() {propertyValues {Curve {keys [<at 0 0>,<at 1 1>]}}}
     Scale Curve=&Lerp_Float(A=0, B=1, Alpha=$System.ReactivityDebug)
    )
    ScaleA=true
    ScaleRGB=true
   )
   &DynamicMaterialParameters(
    Index 0 Param 1=&Lerp_Float(A=0.7, B=0.7, Alpha=$System.ReactivityDebug)
    Index 0 Param 2=&Lerp_Float(A=3.5, B=3.5, Alpha=$System.ReactivityDebug)
    Index 0 Param 3=1
    Index 0 Param 4=1
    Module.Param0WriteEnabled=true
    Module.Param1WriteEnabled=true
    Module.Param2WriteEnabled=true
    Module.Param3WriteEnabled=true
   )
   &SolveForcesAndVelocity(Clamp Velocity=false, Limit Acceleration=false)
   /SetVariables(Particles.BladeTipColor:LinearColor=&MergeHelper_ColorSwitch(
     0=&LerpLinearColors(StartColor=<rgba 0.1173 0.5 0 0>, EndColor=<rgba 0.4693 2 0 0.1>, LerpFactor=$System.ReactivityDebug)
     1=&LerpLinearColors(StartColor=<rgba 0.5 0 0.0239 0>, EndColor=<rgba 2 0 0.0173 0.1>, LerpFactor=$System.ReactivityDebug)
     2=?
     3=?
     Index=$User.Variant
     Num=::NewEnumerator0
    )) {isSetParameters=true}
   /SetVariables(Particles.GlowColor=&MergeHelper_ColorSwitch(
     0=&LerpLinearColors(StartColor=<rgba 0.1173 0.5 0 0>, EndColor=<rgba 0.2714 1 0 0.15>, LerpFactor=$System.ReactivityDebug)
     1=&LerpLinearColors(StartColor=<rgba 0.3542 0 0.0371 0>, EndColor=<rgba 1 0 0.0269 0.2>, LerpFactor=$System.ReactivityDebug)
     2=?
     3=?
     Index=$User.Variant
     Num=::NewEnumerator0
    )) {isSetParameters=true}
  ]
  rendererClasses [&NiagaraMeshRendererProperties]
  renderers [
   {rendererClass=&NiagaraMeshRendererProperties, rendererIndex=0}
   {rendererClass=&NiagaraMeshRendererProperties, rendererIndex=1}
   {rendererClass=&NiagaraMeshRendererProperties, rendererIndex=2}
  ]
 }
 Main_Light {
  emitterUpdateScript [
   &EmitterState(Life Cycle Mode=::NewEnumerator1, Loop Behavior=::NewEnumerator1, Loop Duration=0.1, UseLoopDelay=false)
   &SpawnBurst_Instantaneous(Use Spawn Probability=false, UseLoopCountLimit=false)
  ]
  particleSpawnScript [
   &InitializeParticle(
    Lifetime=2
    Mesh Scale Mode=::NewEnumerator3
    Position Offset=<0 -9 -1.4>
    Sprite Size=<6 12>
    Sprite Size Mode=::NewEnumerator4
    UsePositionOffset=true
    Write Material Random=true
   )
   &SpriteFacingAndAlignment(Sprite Alignment=<0 0 1>, Sprite Facing=<0 0 1>, WriteAlignment=false, WriteFacing=false)
   &CameraOffset(Camera Offset Amount=5)
  ]
  particleUpdateScript [
   &ParticleState(Kill Particles When Lifetime Has Elapsed=false)
   &ScaleColor(
    Scale Alpha=$System.ReactivityDebug
    Scale RGB=&MakeVectorFromLinearColorRGB(Color=&MergeHelper_ColorSwitch(0=<rgba 0.5853 2 0 0.5>, 1=<rgba 2 0.0013 0.1689 0.5>, 2=?, 3=?, Index=$User.Variant, Num=::NewEnumerator0))
    ScaleA=true
    ScaleRGB=true
   )
   /SetVariables(Particles.SpriteSize=&Multiply_Vector2D_ByFloat(
     Float=1
     Vector2D=&Multiply_Vector2D(
      A=$Particles.Initial.SpriteSize
      B=&Lerp_Vector2(
       A=<1 0.5>
       B=<1 1>
       Alpha=&DotProduct(
        Vector A=&ReturnCameraVector(Camera Vectors=::NewEnumerator2, Coordinate Space=::Simulation)
        Vector B=/Game/Effects/Niagara/DynamicInputs/Engine/FN_GetSystemAxis(Axis:ENiagaraOrientationAxis=::ZAxis, Flip Axis:NiagaraBool=false)
       )
      )
     )
    )) {isSetParameters=true}
   &FN_Light_Attributes(Create Position Attribute=true, Position=<0 -15 -1>)
  ]
  rendererClasses [&NiagaraSpriteRendererProperties, &NiagaraLightRendererProperties]
  renderers [
   {rendererClass=&NiagaraSpriteRendererProperties}
   {rendererClass=&NiagaraLightRendererProperties, rendererIndex=1}
  ]
 }
 Wing_Lights {
  emitterUpdateScript [
   &EmitterState(Life Cycle Mode=::NewEnumerator1, Loop Behavior=::NewEnumerator1, Loop Duration=0.1, UseLoopDelay=false)
   &SpawnBurst_Instantaneous(Spawn Count=4, Use Spawn Probability=false, UseLoopCountLimit=false)
   &SpawnBurst_Instantaneous(Spawn Group=1, Use Spawn Probability=false, UseLoopCountLimit=false) {module=SpawnBurst_Instantaneous001}
  ]
  particleSpawnScript [
   /SetVariables(Particles.VisibilityTag=$Emitter.SpawnGroup) {isSetParameters=true}
   &PartitionParticles(Number Of Partitions=4)
   &InitializeParticle(
    Lifetime=2
    Position Offset=&MergeHelper_VectorSwitch(
     0=&Multiply_Vector(
      A=<17 -3 1>
      B=&MergeHelper_VectorSwitch(0=<-1 1 9.7>, 1=<-1 1 -18.7>, 2=<1 1 9.7>, 3=<1 1 -18.7>, Index=$Particles.PartitionParticles.Partition, Num=::NewEnumerator2)
     )
     1=<0 -8 -7>
     2=?
     3=?
     Index=$Emitter.SpawnGroup
     Num=::NewEnumerator0
    )
    Sprite Size=&MergeHelper_Vector2Switch(0=<8 14>, 1=<6 6>, 2=?, 3=?, 4=?, Index=$Emitter.SpawnGroup, Num=::NewEnumerator0)
    Sprite Size Mode=::NewEnumerator4
    UsePositionOffset=false
    Write Material Random=true
   )
  ]
  particleUpdateScript [
   &ParticleState(Kill Particles When Lifetime Has Elapsed=false)
   &ScaleColor(
    Scale Alpha=$System.ReactivityDebug
    Scale RGB=&MakeVectorFromLinearColorRGB(Color=&MergeHelper_ColorSwitch(0=<rgba 0.2926 1 0 0.7>, 1=<rgba 1 0 0.0547 0.7>, 2=?, 3=?, Index=$User.Variant, Num=::NewEnumerator0))
    ScaleA=true
    ScaleRGB=true
   )
   &FN_Light_Attributes(Create Position Attribute=true, Position=<0 -15 -1>)
   &SocketLocation(
    SamplePosOffset=true
    Sampled Position Offset=&MergeHelper_VectorSwitch(
     0=&Multiply_Vector(
      A=<6 0 1>
      B=&MergeHelper_VectorSwitch(0=<-1 1 7>, 1=<-1 1 -7>, 2=<1 1 -7>, 3=<1 1 7>, Index=$Particles.PartitionParticles.Partition, Num=::NewEnumerator2)
     )
     1=<0 -8 -7>
     2=?
     3=?
     Index=$Particles.VisibilityTag
     Num=::NewEnumerator0
    )
    Socket Index=&MergeHelper_IntSwitch(0=$Particles.PartitionParticles.Partition, 1=4, 2=?, 3=?, 4=?, 5=?, Index=$Particles.VisibilityTag, Num=::NewEnumerator0)
    Socket Reader=$System.Socket Reader
    Socket Sampling Mode=::NewEnumerator3
    VelocitySpeedLimit=false
   )
   &SpriteFacingAndAlignment(
    Sprite Alignment=&MergeHelper_VectorSwitch(
     0=&RotateVectorByQuaternion(
      Vector To Rotate=&Multiply_Vector(
       A=<1 1 0.45>
       B=&MergeHelper_VectorSwitch(0=<-1 0 1>, 1=<-1 0 -1>, 2=<1 0 -1>, 3=<1 0 1>, Index=$Particles.PartitionParticles.Partition, Num=::NewEnumerator2)
      )
      Quaternion=$Output.SocketLocation.SampledRotation
     )
     1=<1 0 0>
     2=?
     3=?
     Index=$Emitter.SpawnGroup
     Num=::NewEnumerator0
    )
    Sprite Facing=<0 0 1>
    WriteAlignment=true
    WriteFacing=false
   )
   /SetVariables(Particles.Color_2:LinearColor=&MultiplyLinearColor(
     Linear Color=&MakeLinearColorFromVectorAndFloat("Vector (RGB)"=<1 1 1>, "Float (Alpha)"=$System.ReactivityDebug)
     Scale Factor=&MakeVector4_FromLinearColor(LinearColor=&MergeHelper_ColorSwitch(0=<rgba 0.2926 1 0 0.7>, 1=<rgba 0 0.959 1 0.7>, 2=?, 3=?, Index=$User.Variant, Num=::NewEnumerator0))
    )) {isSetParameters=true}
  ]
  rendererClasses [&NiagaraSpriteRendererProperties]
  renderers [
   {rendererClass=&NiagaraSpriteRendererProperties}
   {rendererClass=&NiagaraSpriteRendererProperties, rendererIndex=1}
  ]
 }
}
summary {
 systemName=NS_Backpack_AshenMagnus
 emitters [
  {emitterName=Main_Light, rendererClasses [&NiagaraSpriteRendererProperties, &NiagaraLightRendererProperties]}
  {emitterName=Wing_Lights, rendererClasses [&NiagaraSpriteRendererProperties]}
  {emitterName=Fans, rendererClasses [&NiagaraMeshRendererProperties]}
 ]
 userVariables [
  User.bIsFrontEnd {type=NiagaraBool}
  User.bIsFrontEndPreview {type=NiagaraBool}
  User.bIsFullyDeployed {type=NiagaraBool}
  User.Reactivity {type=NiagaraFloat, defaultValue=0, description=""}
  User.Variant {type=NiagaraInt32}
 ]
}
