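# Animation Blueprint for the AshenMagnus backpack cosmetic, posing its pack, blade and propeller bones procedurally on top of the input pose.
# A Modify Bone chain offsets the pack, angles four blades, zero-scales four propellers and applies a shared pitch spin, accumulated per update from an interpolated speed.
# The event graph remaps material EmissiveBrightness into Emissive var and watches AthenaKills, zeroing that variable at zero eliminations unless frontend reactivity is shown.
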AnimBlueprint /BRCosmetics/Accessories/FORT_Backpacks/Backpack_M_MED_AshenMagnus/Meshes/M_MED_AshenMagnus_Pack_AnimBP
target_skeleton=&/M_MED_AshenMagnus_Pack_Skeleton
graphs [
 AnimGraph {nodes [
   n1 = "Transform (Modify) Bone - Bone: pack"(ComponentPose=Local To Component(LocalPose=Input Pose()), Translation="0,-3,0", Rotation="0,0,0") {
    rotationMode=BMM_Additive
    rotationSpace=BCS_BoneSpace
    translationMode=BMM_Additive
    translationSpace=BCS_BoneSpace
    alphaBoolBlend {blendInTime=300, blendOption=CircularIn, blendOutTime=300}
    alphaScaleBiasClamp {bInterpResult=true, interpSpeedDecreasing=4, interpSpeedIncreasing=6}
    boneToModify {boneName=pack}
    translation {y=-3}
   }
   n2 = "Transform (Modify) Bone - Bone: blade_R_top"(ComponentPose=@n1, Alpha=1, Translation="0,0,0", Rotation="-60,0,0", Scale="1,1,1") {
    rotationMode=BMM_Additive
    rotationSpace=BCS_BoneSpace
    translationMode=BMM_Additive
    translationSpace=BCS_BoneSpace
    alphaBoolBlend {blendInTime=300, blendOption=CircularIn, blendOutTime=300}
    alphaScaleBiasClamp {bInterpResult=true, interpSpeedDecreasing=4, interpSpeedIncreasing=6}
    boneToModify {boneName=blade_R_top}
    rotation {pitch=-60}
   }
   n3 = "Transform (Modify) Bone - Bone: blade_R_bot"(ComponentPose=@n2, Alpha=1, Translation="0,0,0", Rotation="60,0,0", Scale="1,1,1") {
    rotationMode=BMM_Additive
    rotationSpace=BCS_BoneSpace
    translationMode=BMM_Additive
    translationSpace=BCS_BoneSpace
    alphaBoolBlend {blendInTime=300, blendOption=CircularIn, blendOutTime=300}
    alphaScaleBiasClamp {bInterpResult=true, interpSpeedDecreasing=4, interpSpeedIncreasing=6}
    boneToModify {boneName=blade_R_bot}
    rotation {pitch=60}
   }
   n4 = "Transform (Modify) Bone - Bone: blade_L_bot"(ComponentPose=@n3, Alpha=1, Translation="0,0,0", Rotation="-60,0,0", Scale="1,1,1") {
    rotationMode=BMM_Additive
    rotationSpace=BCS_BoneSpace
    translationMode=BMM_Additive
    translationSpace=BCS_BoneSpace
    alphaBoolBlend {blendInTime=300, blendOption=CircularIn, blendOutTime=300}
    alphaScaleBiasClamp {bInterpResult=true, interpSpeedDecreasing=4, interpSpeedIncreasing=6}
    boneToModify {boneName=blade_L_bot}
    rotation {pitch=-60}
   }
   n5 = "Transform (Modify) Bone - Bone: blade_L_top"(ComponentPose=@n4, Alpha=1, Translation="0,0,0", Rotation="60,0,0", Scale="1,1,1") {
    rotationMode=BMM_Additive
    rotationSpace=BCS_BoneSpace
    translationMode=BMM_Additive
    translationSpace=BCS_BoneSpace
    alphaBoolBlend {blendInTime=300, blendOption=CircularIn, blendOutTime=300}
    alphaScaleBiasClamp {bInterpResult=true, interpSpeedDecreasing=4, interpSpeedIncreasing=6}
    boneToModify {boneName=blade_L_top}
    rotation {pitch=60}
   }
   n6 = "Transform (Modify) Bone - Bone: propeller_R_top"(ComponentPose=@n5, Alpha=1, Translation="0,0,0", Rotation="0,0,0", Scale="0,0,0") {
    rotationMode=BMM_Additive
    rotationSpace=BCS_BoneSpace
    scaleMode=BMM_Replace
    scaleSpace=BCS_BoneSpace
    translationMode=BMM_Additive
    translationSpace=BCS_BoneSpace
    alphaBoolBlend {blendOption=Linear}
    alphaScaleBiasClamp {interpSpeedDecreasing=4, interpSpeedIncreasing=6}
    boneToModify {boneName=propeller_R_top}
    scale=<0 0 0>
   }
   n7 = "Transform (Modify) Bone - Bone: propeller_R_bot"(ComponentPose=@n6, Alpha=1, Translation="0,0,0", Rotation="0,0,0", Scale="0,0,0") {
    rotationMode=BMM_Additive
    rotationSpace=BCS_BoneSpace
    scaleMode=BMM_Replace
    scaleSpace=BCS_BoneSpace
    translationMode=BMM_Additive
    translationSpace=BCS_BoneSpace
    alphaBoolBlend {blendOption=Linear}
    alphaScaleBiasClamp {interpSpeedDecreasing=4, interpSpeedIncreasing=6}
    boneToModify {boneName=propeller_R_bot}
    scale=<0 0 0>
   }
   n8 = "Transform (Modify) Bone - Bone: propeller_L_bot"(ComponentPose=@n7, Alpha=1, Translation="0,0,0", Rotation="0,0,0", Scale="0,0,0") {
    rotationMode=BMM_Additive
    rotationSpace=BCS_BoneSpace
    scaleMode=BMM_Replace
    scaleSpace=BCS_BoneSpace
    translationMode=BMM_Additive
    translationSpace=BCS_BoneSpace
    alphaBoolBlend {blendOption=Linear}
    alphaScaleBiasClamp {interpSpeedDecreasing=4, interpSpeedIncreasing=6}
    boneToModify {boneName=propeller_L_bot}
    scale=<0 0 0>
   }
   n9 = "Transform (Modify) Bone - Bone: propeller_L_top"(ComponentPose=@n8, Alpha=1, Translation="0,0,0", Rotation="0,0,0", Scale="0,0,0") {
    rotationMode=BMM_Additive
    rotationSpace=BCS_BoneSpace
    scaleMode=BMM_Replace
    scaleSpace=BCS_BoneSpace
    translationMode=BMM_Additive
    translationSpace=BCS_BoneSpace
    alphaBoolBlend {blendOption=Linear}
    alphaScaleBiasClamp {interpSpeedDecreasing=4, interpSpeedIncreasing=6}
    boneToModify {boneName=propeller_L_top}
    scale=<0 0 0>
   }
   n10 = "Get Spinning Rotator 00 "() -> ("Spinning Rotator 00 ":Rotator)
   n11 = "Transform (Modify) Bone - Bone: propeller_R_top"(ComponentPose=@n9, Alpha=1, Translation="0,0,0", Rotation=@n10, Scale="1,1,1") {
    rotationMode=BMM_Additive
    rotationSpace=BCS_BoneSpace
    translationMode=BMM_Additive
    translationSpace=BCS_BoneSpace
    alphaBoolBlend {blendInTime=300, blendOption=CircularIn, blendOutTime=300}
    alphaScaleBiasClamp {bInterpResult=true, interpSpeedDecreasing=4, interpSpeedIncreasing=6}
    boneToModify {boneName=propeller_R_top}
   }
   n12 = "Transform (Modify) Bone - Bone: propeller_R_bot"(ComponentPose=@n11, Alpha=1, Translation="0,0,0", Rotation=@n10, Scale="1,1,1") {
    rotationMode=BMM_Additive
    rotationSpace=BCS_BoneSpace
    translationMode=BMM_Additive
    translationSpace=BCS_BoneSpace
    alphaBoolBlend {blendInTime=300, blendOption=CircularIn, blendOutTime=300}
    alphaScaleBiasClamp {bInterpResult=true, interpSpeedDecreasing=4, interpSpeedIncreasing=6}
    boneToModify {boneName=propeller_R_bot}
   }
   n13 = "Transform (Modify) Bone - Bone: propeller_L_bot"(ComponentPose=@n12, Alpha=1, Translation="0,0,0", Rotation=@n10, Scale="1,1,1") {
    rotationMode=BMM_Additive
    rotationSpace=BCS_BoneSpace
    translationMode=BMM_Additive
    translationSpace=BCS_BoneSpace
    alphaBoolBlend {blendInTime=300, blendOption=CircularIn, blendOutTime=300}
    alphaScaleBiasClamp {bInterpResult=true, interpSpeedDecreasing=4, interpSpeedIncreasing=6}
    boneToModify {boneName=propeller_L_bot}
   }
   n14 = "Transform (Modify) Bone - Bone: propeller_L_top"(ComponentPose=@n13, Alpha=1, Translation="0,0,0", Rotation=@n10, Scale="1,1,1") {
    rotationMode=BMM_Additive
    rotationSpace=BCS_BoneSpace
    translationMode=BMM_Additive
    translationSpace=BCS_BoneSpace
    alphaBoolBlend {blendInTime=300, blendOption=CircularIn, blendOutTime=300}
    alphaScaleBiasClamp {bInterpResult=true, interpSpeedDecreasing=4, interpSpeedIncreasing=6}
    boneToModify {boneName=propeller_L_top}
   }
   Output Pose(Result=Component To Local(ComponentPose=@n14)) {layerGroup=None}
  ]}
 EventGraph {nodes [
   n1 = Event BlueprintUpdateAnimation()
   > n2 = Sequence!(execute) -> (then_0, then_1, then_2:Exec, then_3:Exec)
   > n3 = Get Dynamic Material Instance(SkeletalMeshComponent=GetPartSkeletalMeshComponent())
   > n4 = Get Scalar Parameter Value!(execute, self:Material Instance Dynamic Object Reference=@n3, ParameterName:Name=EmissiveBrightness) -> (then, ReturnValue:float) {is=title K2Node_CallFunction}
   > Set Emissive var(Emissive var:float=Remap Float(Input=@n4, OldRange=(X=1.5,Y=2), NewRange=(X=0,Y=1), Clamp=true) -> (Output=~)) -> (Output_Get:float)
   Set Crack Spin Speed(
    execute=@n2.then_1
    Crack Spin Speed:double=FInterpTo(Current=Get Crack Spin Speed() -> (Crack Spin Speed:double), Target=5, DeltaTime=@n1.DeltaTimeX, InterpSpeed=5)
   ) -> (Output_Get:double)
   n5 = Set Crack Spin(
    execute=@n2.then_2
    Crack Spin:double=float + float(A=Get Crack Spin() -> (Crack Spin:double), B=Get Crack Spin Speed()) {is=title K2Node_CommutativeAssociativeBinaryOperator}
   ) -> (Output_Get:double)
   > "Set Spinning Rotator 00 "("Spinning Rotator 00 ":Rotator=MakeRotator(Pitch=NormalizeAxis!(Angle:float=@n5) -> (ReturnValue:float))) -> (Output_Get:Rotator)
   Event OnCharacterCustomizationLoaded()
   > n6 = Cast To FortPlayerPawn(Object=TryGetFortPawn())
   > n7 = WatchCosmeticStat!(
    execute:Exec
    Pawn:Fort Player Pawn Object Reference=@n6
    StatTag:Fort Stat Manager Tag Structure=(TagName="StatManager.AthenaKills")
    DefaultValue:int=0
    bBroadcastInitialValue:bool=true
    InitialValueBroadcastDelay:float=-1
    ChangeValueBroadcastDelay:float=-1
    bConsumeWaitingNotifications:bool=true
   ) -> (then:Exec, OnStatChanged:Exec, StatName:Name, StatValue:int) {is=title K2Node_AsyncAction}
   n8 = Set Elimination Count(execute=@n7.OnStatChanged, Elimination Count:double="To Float (Integer)"(InInt=@n7.StatValue)) -> (Output_Get:double)
   n9 = Get bShouldDisplayFrontendReactivity() -> (bShouldDisplayFrontendReactivity:bool)
   > Branch(Condition=AND Boolean(A="Equal (Float)"(A=@n8), B=NOT Boolean(A=@n9)))
   > Set Emissive var(Emissive var=0)
   Event BlueprintInitializeAnimation()
   > Set Skel_Mesh(Skel_Mesh=GetOwnerSkeletalMeshComponent())
  ]}
 spinFNC_deprecated {nodes [
   spinFNC_deprecated!() -> (then, speed:double=~, initVal:double=~) {is=title K2Node_FunctionEntry}
   > Return Node!(execute, output:double=0)
  ]}
]
variables [
 Active Elimination {type=boolean}
 ActiveKillFX {type=boolean}
 As B Athena Part Modifier Ashen Magnus {type=B_Athena_PartModifier_AshenMagnus_C, default=null}
 Crack Spin {type=number}
 Crack Spin Speed {type=number}
 Elimination Count {type=number}
 Emissive var {type=number}
 Float elim {type=number}
 Reactivity {type=boolean}
 Skel_Mesh {type=SkeletalMeshComponent}
 "Spinning Rotator 00 " {type=Rotator}
 TL_Reactivity {type=TimelineComponent, default=null}
]
